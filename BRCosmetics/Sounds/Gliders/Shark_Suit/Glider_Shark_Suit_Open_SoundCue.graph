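# Sound cue played when the Shark Suit glider opens, layering the open sound with a thrust-start sound.
# A mixer plays the open wave immediately and the thrust-start wave, slightly quieter, after a fixed two-second delay.
# Each branch runs through a single-entry random node.
# Glider open/close attenuation, a cosmetic glider property sheet, the glider land sound class and open/close concurrency govern playback.
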SoundCue /BRCosmetics/Sounds/Gliders/Shark_Suit/Glider_Shark_Suit_Open_SoundCue
root=Mixer_0
nodes [
 {id=Delay_0, children [Random_1], properties {delayMax=2, delayMin=2}}
 {id=Mixer_0, children [Random_0,Delay_0], properties {inputVolume [1,0.8]}}
 {id=Random_0, children [WavePlayer_0], properties {weights [1]}}
 {id=Random_1, children [WavePlayer_1], properties {weights [1]}}
 {id=WavePlayer_0, properties {soundWaveAssetPtr=/Game/Sounds/Fort_Gadgets/Athena/Shark_Suit/Glider_Shark_Suit_Open_01}}
 {id=WavePlayer_1, properties {soundWaveAssetPtr=&/Glider_Shark_Suit_Thrust_Start_01}}
]
properties {
 attenuationSettings=/Game/Sounds/Attenuation/Fort_Player_Glider_OpenClose_Athena_Att
 audioPropertiesSheet=/BRCosmetics/Sounds/Settings/PropertySheets/Glider/APS_Cosmetic_Glider_Open
 maxDistance=10000
 soundClassObject=&Fort_Gliders_Land
 volumeMultiplier=1
 concurrencySet [&Athena_GliderOpenClose_Conc]
}
