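# Sound cue for the looping thrust audio of the Shark Suit glider cosmetic.
# A looping wave player feeds a single-entry Random node into a continuous modulator, which maps the AnyDirection parameter to a slight pitch range and a wider volume range.
# Glider thrust attenuation, a cosmetic glider loop properties sheet and the Fort_Gliders_Land sound class govern playback.
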SoundCue /BRCosmetics/Sounds/Gliders/Shark_Suit/Glider_Shark_Suit_Thrust_Loop_SoundCue
root=ModulatorContinuous_0
nodes [
 {
  id=ModulatorContinuous_0
  children [Random_0]
  properties {
   pitchModulationParams {default=0, maxOutput=1.1, minOutput=0.99, parameterName=AnyDirection}
   volumeModulationParams {default=0, minOutput=0.4, parameterName=AnyDirection}
  }
 }
 {id=Random_0, children [WavePlayer_0], properties {weights [1]}}
 {id=WavePlayer_0, properties {bLooping=true, soundWaveAssetPtr=/Game/Sounds/Fort_Gadgets/Athena/Shark_Suit/Glider_Shark_Suit_Thrust_Loop_01}}
]
properties {
 attenuationSettings=&Fort_Player_Glider_Thrust_Athena_Att
 audioPropertiesSheet=&APS_Cosmetic_Glider_Loop
 maxDistance=12500
 soundClassObject=&Fort_Gliders_Land
 volumeMultiplier=0.9
 concurrencySet [null]
}
